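# Parent material for the TaneSE vehicle wheel cosmetic, adding animated emissive effects to the rim and tire.
# MF_WheelParent_Simple supplies base attributes, with Emissive overridden before the damage and de-rez transition functions.
# Masked, tinted rim glow uses a panned sine pulse, while tire glow uses scrolling cloud noise and an effects mask.
# Vertex color red blends rim to tire; the rim accent-color branch is left unconnected.
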Material /VehicleCosmetics/Wheels/TaneSE/Materials/M_WheelParent_TaneSE
graph {
 properties {
  bAutomaticallySetUsageInEditor=false
  bUseMaterialAttributes=true
  bUsedWithMeshDeformer=true
  bUsedWithSkeletalMesh=true
  pixelDepthOffsetMode=PDOM_Legacy
  usedShadingModels=Default Lit
 }
 nodes [
  n1 = TextureSampleParameter2D() {
   group=5_Special Effects
   parameterName=Rim_Masks
   samplerSource=SSM_Wrap_WorldGroupSettings
   samplerType=SAMPLERTYPE_LinearColor
   texture=/VehicleCosmetics/Wheels/TaneSE/Textures/T_TaneSE_Rim_M
  }
  n2 = VectorParameter() {group=Special_Effect, parameterName=FX_Custom_Color_Emissive, defaultValue=<rgba 0.03 0.4 0.64 1>}
  n3 = Multiply(A=@n1.R, B=@n2.RGB)
  n4 = Panner(Coordinate=TextureCoordinate()) {speedX=0.654}
  n5 = ConstantBiasScale(None=Sine(None=ComponentMask(None=@n4) {r=true})) {bias=-0.1, scale=1}
  n6 = Clamp(None=@n5)
  n7 = TextureCoordinate() {vTiling=3}
  n8 = Constant() {r=0.12}
  n9 = ComponentMask(None=Panner(Coordinate=Add(A=@n7, B=@n8), Time=Constant() {r=0.2}) {speedX=1, speedY=1}) {g=true}
  n10 = Clamp(None=Ceil(None=OneMinus(None=@n9)))
  n11 = Multiply(A=@n3, B=Multiply(A=Constant() {r=2}, B=Multiply(A=@n6, B=@n10)))
  n12 = Clamp(None=Multiply(A=@n5, B=Constant() {r=2})) {maxDefault=3, minDefault=0.4}
  n13 = Multiply(A=@n1.A, B=@n3)
  n14 = Multiply(
   A=Add(A=Multiply(A=@n11, B=@n1.A), B=Multiply(A=@n12, B=@n13))
   B=ScalarParameter() {defaultValue=3, group=Special_Effect, parameterName=Rim_Emissive_Brightness}
  )
  n15 = VectorParameter() {parameterName=FX_Accent_Color_Emissve, defaultValue=<rgba 1 0.46 0.09 1>}
  Add(A=@n14, B=Multiply(A=Multiply(A=@n1.B, B=@n15.RGB), B=ScalarParameter() {defaultValue=3, parameterName=Rim_Accent_EmicciveBrightness}))
  n16 = GetMaterialAttributes(None=MaterialFunctionCall() {materialFunction=&MF_WheelParent_Simple}.MaterialAttributes) {attributeGetTypes [#1]}
  n17 = NamedRerouteUsage()
  MaterialFunctionCall(In=MaterialFunctionCall(Material=SetMaterialAttributes(MaterialAttributes=@n16.MaterialAttributes, Emissive Color=@n17) {attributeSetTypes [#1]}) {materialFunction=&MF_BaseVehicle_Damage}.Result) {materialFunction=&MF_VehicleDeRez_Transition}
  n18 = TextureSampleParameter2D() {
   group=5_Special Effects
   parameterName=Tier_Masks
   samplerSource=SSM_Wrap_WorldGroupSettings
   texture=/VehicleCosmetics/Wheels/TaneSE/Textures/T_TaneSE_Tire_RGB
  }
  n19 = Multiply(A=Multiply(A=@n18.R, B=@n2.RGB), B=ScalarParameter() {defaultValue=3, group=Special_Effect, parameterName=Tire_Emissive_Brightness})
  n20 = TextureCoordinate() {uTiling=3, vTiling=0.15}
  n21 = Power(
   Base=TextureSample(UVs=Panner(Coordinate=@n20) {speedX=0.25, speedY=2}) {samplerSource=SSM_Wrap_WorldGroupSettings, texture=/VehicleCosmetics/Wheels/_SharedTextures/FX/CloudDistort_T}.R
   Exp=Constant() {r=2}
  )
  n22 = TextureCoordinate() {uTiling=4, vTiling=0.25}
  n23 = TextureSample(UVs=Panner(Coordinate=@n22) {speedX=0.1, speedY=0.5}) {samplerSource=SSM_Wrap_WorldGroupSettings, texture=&CloudDistort_T}
  n24 = TextureSample(UVs=Panner(Coordinate=@n22) {speedX=0.1, speedY=0.5}) {
   samplerSource=SSM_Wrap_WorldGroupSettings
   samplerType=SAMPLERTYPE_LinearColor
   texture=/VehicleCosmetics/Wheels/_SharedTextures/FX/CloudyNoise_T
  }
  n25 = Multiply(A=@n24.RGB, B=Constant() {r=3})
  n26 = Clamp(None=Multiply(A=Add(A=@n21, B=@n23.G), B=Add(A=@n25, B=@n23.R)))
  n27 = TextureCoordinate()
  n28 = Constant() {r=0.7}
  n29 = ConstantBiasScale(None=ComponentMask(None=Panner(Coordinate=@n27, Time=@n28) {speedY=-0.2}) {g=true}) {bias=-0.1, scale=0.2}
  n30 = TextureSample() {samplerSource=SSM_Wrap_WorldGroupSettings, texture=/VehicleCosmetics/Wheels/TaneSE/Textures/T_TaneSE_Tire_EffectsMask}
  n31 = Multiply(A=LinearInterpolate(A=Multiply(A=@n26, B=@n29), B=@n30.R, Alpha=@n30.R), B=Constant() {r=3})
  NamedRerouteDeclaration(None=LinearInterpolate(A=@n14, B=Multiply(A=@n19, B=@n31), Alpha=VertexColor().R)) {name=Emissive, nodeColor=<rgba 1 0.75 0 1>}
  OneMinus(None=@n1.A)
 ]
}
