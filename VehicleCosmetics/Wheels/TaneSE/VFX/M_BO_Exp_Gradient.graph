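# Additive unlit Niagara particle material for a soft radial glow in the TaneSE wheel VFX.
# An inverted UV distance from center, raised to a high power, forms the gradient mask.
# A Detail static switch optionally multiplies in cloud noise, panned in radial coordinates at a speed scaled by a dynamic parameter.
# Particle alpha scales the mask into a DepthFade, and particle color is slightly desaturated.
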Material /VehicleCosmetics/Wheels/TaneSE/VFX/M_BO_Exp_Gradient
graph {
 properties {
  bAutomaticallySetUsageInEditor=false
  bUsedWithNiagaraMeshParticles=true
  bUsedWithNiagaraRibbons=true
  bUsedWithNiagaraSprites=true
  blendMode=BLEND_Additive
  previewMesh=/VehicleCosmetics/Wheels/Vented/Mesh/SK_Wheel_Vented
  shadingModel=MSM_Unlit
  usedShadingModels=Unlit
 }
 nodes [
  n1 = Panner(
   Coordinate=Multiply(
    A=Constant2Vector() {g=0.2, r=3}
    B=MaterialFunctionCall() {materialFunction=&VectorToRadialValue}.Radial Coordinates
   )
   Time=Multiply(A=Time(), B=DynamicParameter().Param1)
  ) {speedX=0.2, speedY=1}
  n2 = Add(A=TextureCoordinate(), B=Constant() {r=-0.5})
  n3 = SquareRoot(None=DotProduct(A=@n2, B=@n2))
  n4 = Clamp(None=Power(Base=OneMinus(None=@n3), Exp=Constant() {desc=12, r=12}))
  n5 = Multiply(A=Power(Base=TextureSample(UVs=@n1) {texture=/VehicleCosmetics/Wheels/DemonPortal/Textures/CloudDistort_T}.G, Exp=Constant() {r=4}), B=@n4)
  n6 = ParticleColor()
  DepthFade(Opacity=Multiply(A=StaticSwitchParameter(True=@n5, False=@n4) {parameterName=Detail}, B=@n6.A))
  Desaturation(None=@n6.RGB, Fraction=Constant() {r=0.15}) {luminanceFactors=<rgba 0.3 0.59 0.11 1>}
  Multiply()
 ]
}
